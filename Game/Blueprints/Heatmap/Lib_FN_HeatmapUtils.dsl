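; Heatmap utility library that bins World Partition actor descriptors or raw locations into XY grid cells.
; GetActorDescLocations collects bounds centres of spatially loaded, non-editor actors, filtering by NativeParentClass substring through the Asset Registry.
; SortActorDescsToCells filters by loaded class with cached results, then DoFinalActorDescSort groups descriptors and counts per snapped cell.
; SortLocationsToCells pre-fills a capped grid across the extents, tallies locations per cell and draws debug corner spheres.
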(fn DoFinalActorDescSort
  (params (locations :type Array%20of%20Vectors) (gridSize :type Float%20%28double-precision%29) (actorDescs :type Array%20of%20Actor%20Desc%20Structures))
  (bind _l_struct (LocalVariables|Getl_struct))
  (bind (_descs_15_37b2b7b54d2e6f3e94a2428f9efc8501 _severity_7_12a2f7e746ff0dc731ccf5b3d3d2a202 _location_10_a0565cf1467609f6fe393ea6b2c8bc2f _count_13_07dcd7904e847d07b398d1915f206b67) (Utilities|Struct|BreakHeatmapActorDataStruct _l_struct))
  (bind _l_gridcell (LocalVariables|Getl_gridcell))
  (bind _l_gridcelldata (LocalVariables|Getl_gridCellData))
  (bind (_value _returnvalue) (Utilities|Map|Find _l_gridcelldata _l_gridcell))
  (LocalVariables|Setl_locations locations)
  (LocalVariables|Setl_gridsize gridSize)
  (LocalVariables|Setl_actorDescs actorDescs)
  (for (_array_element _array_index) (LocalVariables|Getl_locations)
    (LocalVariables|Setl_loc _array_element)
    (LocalVariables|Setl_desc (Utilities|Array|Get(acopy) (LocalVariables|Getl_actorDescs) _array_index))
    (LocalVariables|Setl_gridcell (Math|Vector|MakeVector (* (Math|Float|Round (/ (.x (LocalVariables|Getl_loc)) (LocalVariables|Getl_gridsize))) (LocalVariables|Getl_gridsize)) (* (Math|Float|Round (/ (.y (LocalVariables|Getl_loc)) (LocalVariables|Getl_gridsize))) (LocalVariables|Getl_gridsize)) 1.0))
    (if _returnvalue
      (LocalVariables|Setl_struct _value)
      (bind _output_get (LocalVariables|Setl_structDescsCopy _descs_15_37b2b7b54d2e6f3e94a2428f9efc8501))
      (Utilities|Array|Add _output_get (LocalVariables|Getl_desc))
      (bind _structout (Utilities|Struct|SetmembersinHeatmapActorDataStruct (LocalVariables|Getl_struct) (LocalVariables|Getl_structDescsCopy) (Math|Conversions|ToFloat(Integer) (+ _count_13_07dcd7904e847d07b398d1915f206b67 1)) (+ _count_13_07dcd7904e847d07b398d1915f206b67 1)))
      (Utilities|Map|Add (LocalVariables|Getl_gridCellData) (LocalVariables|Getl_gridcell) _structout)
      (else
        (Utilities|Map|Add (LocalVariables|Getl_gridCellData) (LocalVariables|Getl_gridcell) (Utilities|Struct|MakeHeatmapActorDataStruct (Utilities|Array|MakeArray (LocalVariables|Getl_desc)) 1.0 (LocalVariables|Getl_gridcell) 1)))))
  (return (LocalVariables|Getl_gridCellData)))

(fn GetActorDescLocations
  (params (classSubstring :type String))
  (LocalVariables|Setl_class classSubstring)
  (Utilities|FlowControl|Sequence
    (:then_0
      (bind _outactordescs (WorldPartition|GetActorDescs))
      (for _array_element _outactordescs
        (if (.bActorIsEditorOnly (Utilities|Struct|BreakActorDesc _array_element))
          _
          (elif (.bIsSpatiallyLoaded (Utilities|Struct|BreakActorDesc _array_element))
            (if (not (Utilities|String|StartsWith (Utilities|BreakSoftObjectPath (.Class (Utilities|Struct|BreakActorDesc _array_element))) "/Script"))
              (Utilities|Array|Add (LocalVariables|Getl_packageNames) (Utilities|String|StringToName (Utilities|String|Split (Utilities|BreakSoftObjectPath (.Class (Utilities|Struct|BreakActorDesc _array_element))) "." "IgnoreCase" "FromEnd")))
              (Utilities|Array|Add (LocalVariables|Getl_allLocs) (/ (+ (Utilities|Struct|BreakBox (.Bounds (Utilities|Struct|BreakActorDesc _array_element))) (.Max (Utilities|Struct|BreakBox (.Bounds (Utilities|Struct|BreakActorDesc _array_element))))) 2.0)))))))
    (:then_1
      (if (Utilities|String|NotEqualExactly(String) (LocalVariables|Getl_class))
        (for (_array_element_2 _array_index) (LocalVariables|Getl_packageNames)
          (bind _outassetdata (AssetRegistry|GetAssetsbyPackageName (AssetRegistry|GetAssetRegistry) _array_element_2 false true))
          (if (Utilities|Array|IsValidIndex _outassetdata)
            (if (Utilities|String|Contains (AssetData|GetTagValue (Utilities|Array|Get(aref) _outassetdata) (Utilities|Name|MakeLiteralName "NativeParentClass")) (LocalVariables|Getl_class))
              (Utilities|Array|Add (LocalVariables|Getl_locs) (Utilities|Array|Get(aref) (LocalVariables|Getl_allLocs) _array_index)))))
        (else
          (LocalVariables|Setl_locs (LocalVariables|Getl_allLocs)))))
    (:then_2
      (return (LocalVariables|Getl_locs)))))

(fn SnapLocationToGrid
  (params (location :type Vector) (gridSize :type Float%20%28double-precision%29))
  (return (Math|Vector|MakeVector (* (Math|Float|Round (/ (.x location) gridSize)) gridSize) (* (Math|Float|Round (/ (.y location) gridSize)) gridSize) 1.0)))

(fn SortActorDescsToCells
  (params (actorDescs :type Array%20of%20Actor%20Desc%20Structures) (gridSize :type Float%20%28double-precision%29) (classFilter :type Array%20of%20Object%20Class%20References))
  (bind (_guid _nativeclass _class _name _label _bounds) (Utilities|Struct|BreakActorDesc (LocalVariables|Getl_desc)))
  (bind (_min _max) (Utilities|Struct|BreakBox _bounds))
  (bind _l_classeschecked (LocalVariables|Getl_classesChecked))
  (LocalVariables|Setl_actorDescs actorDescs)
  (LocalVariables|Setl_gridsize gridSize)
  (LocalVariables|Setl_classes classFilter)
  (if (Utilities|Array|IsValidIndex (LocalVariables|Getl_actorDescs))
    (Utilities|FlowControl|Sequence
      (:then_0
        (bind _array_element (Utilities|Array|ForEachLoopwithBreak (LocalVariables|Getl_actorDescs))
          (:LoopBody
            (LocalVariables|Setl_desc _array_element)
            (if (not (.bActorIsEditorOnly (Utilities|Struct|BreakActorDesc _array_element)))
              (Utilities|FlowControl|Sequence
                (:then_0
                  (if (Utilities|Array|IsValidIndex (LocalVariables|Getl_classes))
                    (bind _returnvalue (Utilities|LoadClassAssetBlocking (Utilities|ToSoftClassReference (Utilities|MakeSoftClassPath (Utilities|BreakSoftObjectPath (.Class (Utilities|Struct|BreakActorDesc _array_element)))))))
                    (if (.ReturnValue (Utilities|Map|Find _l_classeschecked _returnvalue))
                      (LocalVariables|Setl_allow (Utilities|Map|Find _l_classeschecked _returnvalue))
                      (else
                        (LocalVariables|Setl_allow false)
                        (bind _array_element_2 (Utilities|Array|ForEachLoopwithBreak (LocalVariables|Getl_classes))
                          (:LoopBody
                            (if (Utilities|ClassIsChildOf _returnvalue _array_element_2)
                              (LocalVariables|Setl_allow true)))
                          (:Completed
                            _))
                        (Utilities|Map|Add (LocalVariables|Getl_classesChecked) _returnvalue (LocalVariables|Getl_allow))))
                    (else
                      (LocalVariables|Setl_allow true))))
                (:then_1
                  (if (LocalVariables|Getl_allow)
                    (Utilities|Array|Add (LocalVariables|Getl_locations) (/ (+ _min _max) 2.0))
                    (Utilities|Array|Add (LocalVariables|Getl_finalDescs) (LocalVariables|Getl_desc)))))))))
      (:then_1
        (bind _gridcelldata (Heatmap|DoFinalActorDescSort (LocalVariables|Getl_locations) (LocalVariables|Getl_gridsize) (LocalVariables|Getl_finalDescs)))
        (return _gridcelldata)))))

(fn SortLocationsToCells
  (params (locations :type Array%20of%20Vectors) (gridSize :type Float%20%28double-precision%29))
  (bind _l_gridcell (LocalVariables|Getl_gridcell))
  (bind _l_cells (LocalVariables|Getl_cells))
  (bind (_value _returnvalue) (Utilities|Map|Find _l_cells _l_gridcell))
  (bind _l_minvector (LocalVariables|Getl_minVector))
  (bind _x (.x _l_minvector))
  (bind _l_maxvector (LocalVariables|Getl_maxVector))
  (bind _x_2 (.x _l_maxvector))
  (bind _returnvalue_2 (- _x_2 _x))
  (bind _returnvalue_3 (Math|Float|Absolute(Float) _returnvalue_2))
  (bind _l_gridsize (LocalVariables|Getl_gridsize))
  (bind _returnvalue_4 (* _l_gridsize 1.2))
  (bind _l_gridsize_2 (LocalVariables|Getl_gridsize))
  (bind _returnvalue_5 (/ _returnvalue_3 _l_gridsize_2))
  (bind (_l_minvector_x _l_minvector_y) (Math|Vector|BreakVector (Variables|Getl_minVector_X)))
  (bind (_l_maxvector_x _l_maxvector_y) (Math|Vector|BreakVector (Variables|Getl_maxVector_X)))
  (bind _l_sizex (LocalVariables|Getl_sizeX))
  (LocalVariables|Setl_locations locations)
  (LocalVariables|Setl_gridsize gridSize)
  (Rendering|Debug|FlushPersistentDebugLines 0)
  (Utilities|FlowControl|Sequence
    (:then_0
      (for _array_element (LocalVariables|Getl_locations)
        (LocalVariables|Setl_minVector (Math|Vector|MakeVector (Math|Float|Min(Float) (.x _array_element) _l_minvector_x) (Math|Float|Min(Float) (.y _array_element) _l_minvector_y) 5e+01))
        (LocalVariables|Setl_maxVector (Math|Vector|MakeVector (Math|Float|Max(Float) (.x _array_element) _l_maxvector_x) (Math|Float|Max(Float) (.y _array_element) _l_maxvector_y) 5e+01)))
      (LocalVariables|Setl_minVector (CallFunction|SnapLocationtoGrid (LocalVariables|Getl_minVector) (LocalVariables|Getl_gridsize)))
      (LocalVariables|Setl_maxVector (CallFunction|SnapLocationtoGrid (LocalVariables|Getl_maxVector) (LocalVariables|Getl_gridsize)))
      (Rendering|Debug|DrawDebugSphere 0 (LocalVariables|Getl_minVector) _returnvalue_4 12 "(R=0.000000,G=0.520833,B=0.014202,A=1.000000)" 22.0 22.0)
      (Rendering|Debug|DrawDebugSphere 0 (LocalVariables|Getl_maxVector) _returnvalue_4 12 "(R=0.317661,G=0.520833,B=0.000000,A=1.000000)" 22.0 22.0))
    (:then_1
      (LocalVariables|Setl_curVector (LocalVariables|Getl_minVector))
      (bind _output_get (LocalVariables|Setl_total (Math|Integer|Clamp(Integer) (Math|Float|Ceil (* _returnvalue_5 (/ (Math|Float|Absolute(Float) (- (.y _l_maxvector) (.y _l_minvector))) (LocalVariables|Getl_gridsize)))) 0 1000)))
      (LocalVariables|Setl_sizeX (Math|Float|Floor _returnvalue_5))
      (for _index (range _output_get)
        (LocalVariables|Setl_row (/ _index _l_sizex))
        (LocalVariables|Setl_column (Math|Integer|%25%28Integer%29 _index _l_sizex))
        (Utilities|Map|Add (LocalVariables|Getl_cells) (CallFunction|SnapLocationtoGrid (+ (LocalVariables|Getl_minVector) (Math|Vector|MakeVector (* (LocalVariables|Getl_column) (LocalVariables|Getl_gridsize)) (* (LocalVariables|Getl_row) (LocalVariables|Getl_gridsize)))) (LocalVariables|Getl_gridsize)))))
    (:then_2
      (for _array_element_2 (LocalVariables|Getl_locations)
        (LocalVariables|Setl_loc _array_element_2)
        (LocalVariables|Setl_gridcell (CallFunction|SnapLocationtoGrid (LocalVariables|Getl_loc) (LocalVariables|Getl_gridsize)))
        (if _returnvalue
          (Utilities|Map|Add (LocalVariables|Getl_cells) (LocalVariables|Getl_gridcell) (+ _value 1.0))
          (else
            (Utilities|Map|Add (LocalVariables|Getl_cells) (LocalVariables|Getl_gridcell) 1.0)))))
    (:then_3
      (return (LocalVariables|Getl_cells)))))
